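# Conversation database for the AlmondSplash NPC, a Winterfest-themed Snoop character, defining its top-level greeting and menu.
# The entry point leads to a Speech node with a default holiday line, plus prioritized contextual lines for a first-ever conversation and for players wearing one of two LemonCart outfits.
# The greeting fans out through link nodes to shared NPC quests, basic services, two Season 33 Winterfest quest parts and the character's own dialogue branch.
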ConversationDatabase /NPCLibrary/NPCs/AlmondSplash/NPCLibrary_Conversation_AlmondSplash
entry_points [{tag=Conversation.Entry.AlmondSplash, destinations [#1]}]
nodes [
 ConversationEntryPointNode_7 {
  id=#1
  class=ConversationEntryPointNode
  outputs [#2]
  properties {entryTag {tagName=Conversation.Entry.AlmondSplash}}
 }
 Speech_C_4 {
  id=#2
  class=Speech_C
  outputs [
   #3
   #4
   #5
   #6
   #7
  ]
  properties {
   nodeName=""
   generalConfig {
    defaultMessage=Happy holi-doggs.
    contextualMessages [
     {
      message=Y'all ready for Snoop on a Stoop? All I'm missing is my stoop.
      contextRequirements [{requirement {class=FortControllerRequirement_IsFirstEverConversationWithNPC}}]
      priority {curve {rowName=Tandem.Speech.Priority.FirstEverInteraction}}
     }
     {
      message=Where's your Winterfest spirit?
      contextRequirements [{requirement {
         class=FortControllerRequirement_HasCID
         properties {allowedCIds [
           /BRCosmetics/Athena/Items/Cosmetics/Characters/Character_LemonCartGranite
           /BRCosmetics/Athena/Items/Cosmetics/Characters/Character_LemonCartCleaner
          ]}
        }}]
      priority {curve {rowName=Tandem.Speech.Priority.OutfitReaction}}
     }
    ]
   }
  }
 }
 ConversationLinkNode_4 {
  id=#7
  class=ConversationLinkNode
  properties {remoteEntryTag {tagName=Conversation.Entry.AlmondSplash.Dialogue}}
 }
 ConversationLinkNode_3 {
  id=#6
  class=ConversationLinkNode
  properties {remoteEntryTag {tagName=Conversation.Entry.Season33.WF.Quest.01.part.05}}
 }
 ConversationLinkNode_2 {
  id=#5
  class=ConversationLinkNode
  properties {remoteEntryTag {tagName=Conversation.Entry.Season33.WF.Quest.01.part.04}}
 }
 ConversationLinkNode_0 {id=#4, class=ConversationLinkNode, properties {remoteEntryTag {tagName=Conversation.Entry.BasicServices}}}
 ConversationLinkNode_1 {id=#3, class=ConversationLinkNode, properties {remoteEntryTag {tagName=Conversation.Entry.NPCLibrary.Quests}}}
]
